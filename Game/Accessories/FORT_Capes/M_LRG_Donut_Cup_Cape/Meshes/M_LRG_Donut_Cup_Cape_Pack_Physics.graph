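# Physics asset for the Donut Cup cape accessory, defining collision and joints for the cape's dynamic bones.
# Five chains of box bodies hang from tiny MSTR anchors on spine_05, with limited swing and twist down each chain.
# Free-rotation cross constraints link neighboring chains row by row, forming a cloth-like lattice.
# Capsules on torso, arms and legs, plus a wide flat root box below, provide body and apparent floor collision shapes.
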PhysicsAsset /Game/Accessories/FORT_Capes/M_LRG_Donut_Cup_Cape/Meshes/M_LRG_Donut_Cup_Cape_Pack_Physics
bodies {
 calf_l [{shapeType=Capsule, length=30, radius=12, center=<-16.1768 0.0007 0.0004>, rotation=<rot -90 0 -179.9999>}]
 calf_r [{shapeType=Capsule, length=30, radius=12, center=<16.1768 -0.0007 -0.0004>, rotation=<rot -90 0 0.0001>}]
 clavicle_l [{shapeType=Capsule, length=18, radius=10.75, center=<9.2074 -1.6511 -1.7917>, rotation=<rot -90 0 0.0001>}]
 clavicle_r [{shapeType=Capsule, length=17.5, radius=11.25, center=<-9.2074 1.6511 1.7917>, rotation=<rot -90 0 -179.9999>}]
 dyn_cape_01 [{shapeType=Box, extentX=13, extentY=19.5, extentZ=3, center=<6.5042 0 0>}]
 dyn_cape_010 [{shapeType=Box, extentX=13, extentY=19.5, extentZ=3, center=<6.5042 0 0>}]
 dyn_cape_010_l [{shapeType=Box, extentX=13, extentY=19.5, extentZ=3, center=<6.5042 0 0>}]
 dyn_cape_010_mid_l [{shapeType=Box, extentX=13, extentY=19.5, extentZ=3, center=<6.5042 0 0>}]
 dyn_cape_010_mid_r [{shapeType=Box, extentX=13, extentY=19.5, extentZ=3, center=<6.5042 0 0>}]
 dyn_cape_010_r [{shapeType=Box, extentX=13, extentY=19.5, extentZ=3, center=<6.5042 0 0>}]
 dyn_cape_011 [{shapeType=Box, extentX=13, extentY=19.5, extentZ=3, center=<6.5042 0 0>}]
 dyn_cape_011_l [{shapeType=Box, extentX=13, extentY=19.5, extentZ=3, center=<6.5042 0 0>}]
 dyn_cape_011_mid_l [{shapeType=Box, extentX=13, extentY=19.5, extentZ=3, center=<6.5042 0 0>}]
 dyn_cape_01_MSTR [{shapeType=Capsule, length=0.01, radius=0.01}]
 dyn_cape_01_l [{shapeType=Box, extentX=13, extentY=19.5, extentZ=3, center=<6.5042 0 0>}]
 dyn_cape_01_l_MSTR [{shapeType=Capsule, length=0.01, radius=0.01}]
 dyn_cape_01_mid_l [{shapeType=Box, extentX=13, extentY=19.5, extentZ=3, center=<6.5042 0 0>}]
 dyn_cape_01_mid_l_MSTR [{shapeType=Capsule, length=0.01, radius=0.01}]
 dyn_cape_01_mid_r [{shapeType=Box, extentX=13, extentY=19.5, extentZ=3, center=<6.5042 0 0>}]
 dyn_cape_01_mid_r_MSTR [{shapeType=Capsule, length=0.01, radius=0.01}]
 dyn_cape_01_r [{shapeType=Box, extentX=13, extentY=19.5, extentZ=3, center=<6.5042 0 0>}]
 dyn_cape_01_r_MSTR [{shapeType=Capsule, length=0.01, radius=0.01}]
 dyn_cape_02 [{shapeType=Box, extentX=13, extentY=19.5, extentZ=3, center=<6.5042 0 0>}]
 dyn_cape_02_l [{shapeType=Box, extentX=13, extentY=19.5, extentZ=3, center=<6.5042 0 0>}]
 dyn_cape_02_mid_l [{shapeType=Box, extentX=13, extentY=19.5, extentZ=3, center=<6.5042 0 0>}]
 dyn_cape_02_mid_r [{shapeType=Box, extentX=13, extentY=19.5, extentZ=3, center=<6.5042 0 0>}]
 dyn_cape_02_r [{shapeType=Box, extentX=13, extentY=19.5, extentZ=3, center=<6.5042 0 0>}]
 dyn_cape_03 [{shapeType=Box, extentX=13, extentY=19.5, extentZ=3, center=<6.5042 0 0>}]
 dyn_cape_03_l [{shapeType=Box, extentX=13, extentY=19.5, extentZ=3, center=<6.5042 0 0>}]
 dyn_cape_03_mid_l [{shapeType=Box, extentX=13, extentY=19.5, extentZ=3, center=<6.5042 0 0>}]
 dyn_cape_03_mid_r [{shapeType=Box, extentX=13, extentY=19.5, extentZ=3, center=<6.5042 0 0>}]
 dyn_cape_03_r [{shapeType=Box, extentX=13, extentY=19.5, extentZ=3, center=<6.5042 0 0>}]
 dyn_cape_04 [{shapeType=Box, extentX=13, extentY=19.5, extentZ=3, center=<6.5042 0 0>}]
 dyn_cape_04_l [{shapeType=Box, extentX=13, extentY=19.5, extentZ=3, center=<6.5042 0 0>}]
 dyn_cape_04_mid_l [{shapeType=Box, extentX=13, extentY=19.5, extentZ=3, center=<6.5042 0 0>}]
 dyn_cape_04_mid_r [{shapeType=Box, extentX=13, extentY=19.5, extentZ=3, center=<6.5042 0 0>}]
 dyn_cape_04_r [{shapeType=Box, extentX=13, extentY=19.5, extentZ=3, center=<6.5042 0 0>}]
 dyn_cape_05 [{shapeType=Box, extentX=13, extentY=19.5, extentZ=3, center=<6.5042 0 0>}]
 dyn_cape_05_l [{shapeType=Box, extentX=13, extentY=19.5, extentZ=3, center=<6.5042 0 0>}]
 dyn_cape_05_mid_l [{shapeType=Box, extentX=13, extentY=19.5, extentZ=3, center=<6.5042 0 0>}]
 dyn_cape_05_mid_r [{shapeType=Box, extentX=13, extentY=19.5, extentZ=3, center=<6.5042 0 0>}]
 dyn_cape_05_r [{shapeType=Box, extentX=13, extentY=19.5, extentZ=3, center=<6.5042 0 0>}]
 dyn_cape_06 [{shapeType=Box, extentX=13, extentY=19.5, extentZ=3, center=<6.5042 0 0>}]
 dyn_cape_06_l [{shapeType=Box, extentX=13, extentY=19.5, extentZ=3, center=<6.5042 0 0>}]
 dyn_cape_06_mid_l [{shapeType=Box, extentX=13, extentY=19.5, extentZ=3, center=<6.5042 0 0>}]
 dyn_cape_06_mid_r [{shapeType=Box, extentX=13, extentY=19.5, extentZ=3, center=<6.5042 0 0>}]
 dyn_cape_06_r [{shapeType=Box, extentX=13, extentY=19.5, extentZ=3, center=<6.5042 0 0>}]
 dyn_cape_07 [{shapeType=Box, extentX=13, extentY=19.5, extentZ=3, center=<6.5042 0 0>}]
 dyn_cape_07_l [{shapeType=Box, extentX=13, extentY=19.5, extentZ=3, center=<6.5042 0 0>}]
 dyn_cape_07_mid_l [{shapeType=Box, extentX=13, extentY=19.5, extentZ=3, center=<6.5042 0 0>}]
 dyn_cape_07_mid_r [{shapeType=Box, extentX=13, extentY=19.5, extentZ=3, center=<6.5042 0 0>}]
 dyn_cape_07_r [{shapeType=Box, extentX=13, extentY=19.5, extentZ=3, center=<6.5042 0 0>}]
 dyn_cape_08 [{shapeType=Box, extentX=13, extentY=19.5, extentZ=3, center=<6.5042 0 0>}]
 dyn_cape_08_l [{shapeType=Box, extentX=13, extentY=19.5, extentZ=3, center=<6.5042 0 0>}]
 dyn_cape_08_mid_l [{shapeType=Box, extentX=13, extentY=19.5, extentZ=3, center=<6.5042 0 0>}]
 dyn_cape_08_mid_r [{shapeType=Box, extentX=13, extentY=19.5, extentZ=3, center=<6.5042 0 0>}]
 dyn_cape_08_r [{shapeType=Box, extentX=13, extentY=19.5, extentZ=3, center=<6.5042 0 0>}]
 dyn_cape_09 [{shapeType=Box, extentX=13, extentY=19.5, extentZ=3, center=<6.5042 0 0>}]
 dyn_cape_09_l [{shapeType=Box, extentX=13, extentY=19.5, extentZ=3, center=<6.5042 0 0>}]
 dyn_cape_09_mid_l [{shapeType=Box, extentX=13, extentY=19.5, extentZ=3, center=<6.5042 0 0>}]
 dyn_cape_09_mid_r [{shapeType=Box, extentX=13, extentY=19.5, extentZ=3, center=<6.5042 0 0>}]
 dyn_cape_09_r [{shapeType=Box, extentX=13, extentY=19.5, extentZ=3, center=<6.5042 0 0>}]
 lowerarm_l [{shapeType=Capsule, length=20, radius=8.75, center=<8.797 -0.0005 0>, rotation=<rot -90 0 0.0001>}]
 lowerarm_r [{shapeType=Capsule, length=20, radius=8.75, center=<-8.797 0.0005 0>, rotation=<rot -90 0 -179.9999>}]
 pelvis [{shapeType=Capsule, length=12.75, radius=17.75}]
 root [{shapeType=Box, extentX=201.5, extentY=175.5, extentZ=17.5, center=<-0.0001 0 -9.1006>, rotation=<rot -2.7875 -90 0>}]
 spine_01 [{shapeType=Capsule, length=14.75, radius=15.75}]
 spine_02 [{shapeType=Capsule, length=12.5, radius=16.25, center=<0 -0.9313 0>}]
 spine_03 [{shapeType=Capsule, length=13, radius=15.75}]
 spine_04 [{shapeType=Capsule, length=13, radius=15.75, center=<0 -1.0817 0>}]
 spine_05 [{shapeType=Capsule, length=28.75, radius=16.25, center=<-2.4973 -0.7262 -0.0009>}]
 thigh_l [
  {shapeType=Capsule, length=39.5, radius=12, center=<-20.9405 0.0005 0.0004>, rotation=<rot -90 0 -179.9999>}
  {shapeType=Capsule, length=22, radius=8, center=<-8.9502 -4.0673 -5.7169>, rotation=<rot -90 0 -30.0037>}
 ]
 thigh_r [{shapeType=Capsule, length=39.5, radius=12, center=<20.9405 -0.0005 -0.0004>, rotation=<rot -90 0 0.0001>}]
 upperarm_l [{shapeType=Capsule, length=17.5, radius=11.25, center=<8.797 -0.0005 0>, rotation=<rot -90 0 0.0001>}]
 upperarm_r [{shapeType=Capsule, length=17.25, radius=11.5, center=<-8.797 0.0005 0>, rotation=<rot -90 0 -179.9999>}]
}
constraints [
 {bone1Name=dyn_cape_01_l, bone2Name=dyn_cape_01_l_MSTR, swing1LimitDegrees=45, swing2LimitDegrees=45, twistLimitDegrees=30}
 {bone1Name=dyn_cape_02_l, bone2Name=dyn_cape_01_l, swing1LimitDegrees=45, swing2LimitDegrees=45, twistLimitDegrees=5}
 {bone1Name=dyn_cape_03_l, bone2Name=dyn_cape_02_l, swing1LimitDegrees=45, swing2LimitDegrees=45, twistLimitDegrees=30}
 {bone1Name=dyn_cape_04_l, bone2Name=dyn_cape_03_l, swing1LimitDegrees=45, swing2LimitDegrees=45, twistLimitDegrees=30}
 {bone1Name=dyn_cape_05_l, bone2Name=dyn_cape_04_l, swing1LimitDegrees=45, swing2LimitDegrees=45, twistLimitDegrees=30}
 {bone1Name=dyn_cape_06_l, bone2Name=dyn_cape_05_l, swing1LimitDegrees=45, swing2LimitDegrees=45, twistLimitDegrees=30}
 {bone1Name=dyn_cape_07_l, bone2Name=dyn_cape_06_l, swing1LimitDegrees=45, swing2LimitDegrees=45, twistLimitDegrees=30}
 {bone1Name=dyn_cape_08_l, bone2Name=dyn_cape_07_l, swing1LimitDegrees=45, swing2LimitDegrees=45, twistLimitDegrees=30}
 {bone1Name=dyn_cape_09_l, bone2Name=dyn_cape_08_l, swing1LimitDegrees=45, swing2LimitDegrees=45, twistLimitDegrees=30}
 {
  bone1Name=dyn_cape_01_l_MSTR
  bone2Name=spine_05
  swing1LimitDegrees=45
  swing1Motion=Free
  swing2LimitDegrees=45
  swing2Motion=Free
  twistLimitDegrees=45
  twistMotion=Free
 }
 {
  bone1Name=spine_01
  bone2Name=pelvis
  swing1LimitDegrees=45
  swing1Motion=Free
  swing2LimitDegrees=45
  swing2Motion=Free
  twistLimitDegrees=45
  twistMotion=Free
 }
 {
  bone1Name=spine_02
  bone2Name=spine_01
  swing1LimitDegrees=45
  swing1Motion=Free
  swing2LimitDegrees=45
  swing2Motion=Free
  twistLimitDegrees=45
  twistMotion=Free
 }
 {
  bone1Name=spine_03
  bone2Name=spine_02
  swing1LimitDegrees=45
  swing1Motion=Free
  swing2LimitDegrees=45
  swing2Motion=Free
  twistLimitDegrees=45
  twistMotion=Free
 }
 {
  bone1Name=spine_04
  bone2Name=spine_03
  swing1LimitDegrees=45
  swing1Motion=Free
  swing2LimitDegrees=45
  swing2Motion=Free
  twistLimitDegrees=45
  twistMotion=Free
 }
 {
  bone1Name=spine_05
  bone2Name=spine_04
  swing1LimitDegrees=45
  swing1Motion=Free
  swing2LimitDegrees=45
  swing2Motion=Free
  twistLimitDegrees=45
  twistMotion=Free
 }
 {
  bone1Name=clavicle_l
  bone2Name=spine_05
  swing1LimitDegrees=45
  swing1Motion=Free
  swing2LimitDegrees=45
  swing2Motion=Free
  twistLimitDegrees=45
  twistMotion=Free
 }
 {
  bone1Name=thigh_l
  bone2Name=pelvis
  swing1LimitDegrees=45
  swing1Motion=Free
  swing2LimitDegrees=45
  swing2Motion=Free
  twistLimitDegrees=45
  twistMotion=Free
 }
 {
  bone1Name=calf_l
  bone2Name=thigh_l
  swing1LimitDegrees=45
  swing1Motion=Free
  swing2LimitDegrees=45
  swing2Motion=Free
  twistLimitDegrees=45
  twistMotion=Free
 }
 {
  bone1Name=calf_r
  bone2Name=thigh_r
  swing1LimitDegrees=45
  swing1Motion=Free
  swing2LimitDegrees=45
  swing2Motion=Free
  twistLimitDegrees=45
  twistMotion=Free
 }
 {
  bone1Name=thigh_r
  bone2Name=pelvis
  swing1LimitDegrees=45
  swing1Motion=Free
  swing2LimitDegrees=45
  swing2Motion=Free
  twistLimitDegrees=45
  twistMotion=Free
 }
 {
  bone1Name=upperarm_l
  bone2Name=clavicle_l
  swing1LimitDegrees=45
  swing1Motion=Free
  swing2LimitDegrees=45
  swing2Motion=Free
  twistLimitDegrees=45
  twistMotion=Free
 }
 {
  bone1Name=lowerarm_l
  bone2Name=upperarm_l
  swing1LimitDegrees=45
  swing1Motion=Free
  swing2LimitDegrees=45
  swing2Motion=Free
  twistLimitDegrees=45
  twistMotion=Free
 }
 {
  bone1Name=upperarm_r
  bone2Name=clavicle_r
  swing1LimitDegrees=45
  swing1Motion=Free
  swing2LimitDegrees=45
  swing2Motion=Free
  twistLimitDegrees=45
  twistMotion=Free
 }
 {
  bone1Name=lowerarm_r
  bone2Name=upperarm_r
  swing1LimitDegrees=45
  swing1Motion=Free
  swing2LimitDegrees=45
  swing2Motion=Free
  twistLimitDegrees=45
  twistMotion=Free
 }
 {
  bone1Name=clavicle_r
  bone2Name=spine_05
  swing1LimitDegrees=45
  swing1Motion=Free
  swing2LimitDegrees=45
  swing2Motion=Free
  twistLimitDegrees=45
  twistMotion=Free
 }
 {bone1Name=dyn_cape_010_l, bone2Name=dyn_cape_09_l, swing1LimitDegrees=45, swing2LimitDegrees=45, twistLimitDegrees=30}
 {
  bone1Name=dyn_cape_01_r_MSTR
  bone2Name=spine_05
  swing1LimitDegrees=45
  swing1Motion=Free
  swing2LimitDegrees=45
  swing2Motion=Free
  twistLimitDegrees=45
  twistMotion=Free
 }
 {bone1Name=dyn_cape_01_r, bone2Name=dyn_cape_01_r_MSTR, swing1LimitDegrees=45, swing2LimitDegrees=45, twistLimitDegrees=30}
 {bone1Name=dyn_cape_02_r, bone2Name=dyn_cape_01_r, swing1LimitDegrees=45, swing2LimitDegrees=45, twistLimitDegrees=30}
 {bone1Name=dyn_cape_03_r, bone2Name=dyn_cape_02_r, swing1LimitDegrees=45, swing2LimitDegrees=45, twistLimitDegrees=30}
 {bone1Name=dyn_cape_04_r, bone2Name=dyn_cape_03_r, swing1LimitDegrees=45, swing2LimitDegrees=45, twistLimitDegrees=30}
 {bone1Name=dyn_cape_05_r, bone2Name=dyn_cape_04_r, swing1LimitDegrees=45, swing2LimitDegrees=45, twistLimitDegrees=30}
 {bone1Name=dyn_cape_06_r, bone2Name=dyn_cape_05_r, swing1LimitDegrees=45, swing2LimitDegrees=45, twistLimitDegrees=30}
 {bone1Name=dyn_cape_07_r, bone2Name=dyn_cape_06_r, swing1LimitDegrees=45, swing2LimitDegrees=45, twistLimitDegrees=30}
 {bone1Name=dyn_cape_08_r, bone2Name=dyn_cape_07_r, swing1LimitDegrees=45, swing2LimitDegrees=45, twistLimitDegrees=30}
 {bone1Name=dyn_cape_09_r, bone2Name=dyn_cape_08_r, swing1LimitDegrees=45, swing2LimitDegrees=45, twistLimitDegrees=30}
 {bone1Name=dyn_cape_010_r, bone2Name=dyn_cape_09_r, swing1LimitDegrees=45, swing2LimitDegrees=45, twistLimitDegrees=30}
 {
  bone1Name=dyn_cape_01_mid_r_MSTR
  bone2Name=spine_05
  swing1LimitDegrees=45
  swing1Motion=Free
  swing2LimitDegrees=45
  swing2Motion=Free
  twistLimitDegrees=45
  twistMotion=Free
 }
 {bone1Name=dyn_cape_01_mid_r, bone2Name=dyn_cape_01_mid_r_MSTR, swing1LimitDegrees=20, swing2LimitDegrees=20, twistLimitDegrees=20}
 {bone1Name=dyn_cape_02_mid_r, bone2Name=dyn_cape_01_mid_r, swing1LimitDegrees=45, swing2LimitDegrees=45, twistLimitDegrees=30}
 {bone1Name=dyn_cape_03_mid_r, bone2Name=dyn_cape_02_mid_r, swing1LimitDegrees=45, swing2LimitDegrees=45, twistLimitDegrees=30}
 {bone1Name=dyn_cape_04_mid_r, bone2Name=dyn_cape_03_mid_r, swing1LimitDegrees=45, swing2LimitDegrees=45, twistLimitDegrees=30}
 {bone1Name=dyn_cape_05_mid_r, bone2Name=dyn_cape_04_mid_r, swing1LimitDegrees=45, swing2LimitDegrees=45, twistLimitDegrees=30}
 {bone1Name=dyn_cape_06_mid_r, bone2Name=dyn_cape_05_mid_r, swing1LimitDegrees=45, swing2LimitDegrees=45, twistLimitDegrees=30}
 {bone1Name=dyn_cape_07_mid_r, bone2Name=dyn_cape_06_mid_r, swing1LimitDegrees=45, swing2LimitDegrees=45, twistLimitDegrees=30}
 {bone1Name=dyn_cape_08_mid_r, bone2Name=dyn_cape_07_mid_r, swing1LimitDegrees=45, swing2LimitDegrees=45, twistLimitDegrees=30}
 {bone1Name=dyn_cape_09_mid_r, bone2Name=dyn_cape_08_mid_r, swing1LimitDegrees=45, swing2LimitDegrees=45, twistLimitDegrees=30}
 {bone1Name=dyn_cape_010_mid_r, bone2Name=dyn_cape_09_mid_r, swing1LimitDegrees=45, swing2LimitDegrees=45, twistLimitDegrees=30}
 {
  bone1Name=dyn_cape_01_MSTR
  bone2Name=spine_05
  swing1LimitDegrees=45
  swing1Motion=Free
  swing2LimitDegrees=45
  swing2Motion=Free
  twistLimitDegrees=45
  twistMotion=Free
 }
 {bone1Name=dyn_cape_01, bone2Name=dyn_cape_01_MSTR, swing1LimitDegrees=20, swing2LimitDegrees=20, twistLimitDegrees=20}
 {bone1Name=dyn_cape_02, bone2Name=dyn_cape_01, swing1LimitDegrees=45, swing2LimitDegrees=45, twistLimitDegrees=30}
 {bone1Name=dyn_cape_03, bone2Name=dyn_cape_02, swing1LimitDegrees=45, swing2LimitDegrees=45, twistLimitDegrees=30}
 {bone1Name=dyn_cape_04, bone2Name=dyn_cape_03, swing1LimitDegrees=45, swing2LimitDegrees=45, twistLimitDegrees=30}
 {bone1Name=dyn_cape_05, bone2Name=dyn_cape_04, swing1LimitDegrees=45, swing2LimitDegrees=45, twistLimitDegrees=30}
 {bone1Name=dyn_cape_06, bone2Name=dyn_cape_05, swing1LimitDegrees=45, swing2LimitDegrees=45, twistLimitDegrees=30}
 {bone1Name=dyn_cape_07, bone2Name=dyn_cape_06, swing1LimitDegrees=45, swing2LimitDegrees=45, twistLimitDegrees=30}
 {bone1Name=dyn_cape_08, bone2Name=dyn_cape_07, swing1LimitDegrees=45, swing2LimitDegrees=45, twistLimitDegrees=30}
 {bone1Name=dyn_cape_09, bone2Name=dyn_cape_08, swing1LimitDegrees=45, swing2LimitDegrees=45, twistLimitDegrees=30}
 {bone1Name=dyn_cape_010, bone2Name=dyn_cape_09, swing1LimitDegrees=45, swing2LimitDegrees=45, twistLimitDegrees=30}
 {bone1Name=dyn_cape_011, bone2Name=dyn_cape_010, swing1LimitDegrees=45, swing2LimitDegrees=45, twistLimitDegrees=30}
 {
  bone1Name=dyn_cape_01_mid_l_MSTR
  bone2Name=spine_05
  swing1LimitDegrees=45
  swing1Motion=Free
  swing2LimitDegrees=45
  swing2Motion=Free
  twistLimitDegrees=45
  twistMotion=Free
 }
 {bone1Name=dyn_cape_01_mid_l, bone2Name=dyn_cape_01_mid_l_MSTR, swing1LimitDegrees=20, swing2LimitDegrees=20, twistLimitDegrees=20}
 {bone1Name=dyn_cape_02_mid_l, bone2Name=dyn_cape_01_mid_l, swing1LimitDegrees=45, swing2LimitDegrees=45, twistLimitDegrees=30}
 {bone1Name=dyn_cape_03_mid_l, bone2Name=dyn_cape_02_mid_l, swing1LimitDegrees=45, swing2LimitDegrees=45, twistLimitDegrees=30}
 {bone1Name=dyn_cape_04_mid_l, bone2Name=dyn_cape_03_mid_l, swing1LimitDegrees=45, swing2LimitDegrees=45, twistLimitDegrees=30}
 {bone1Name=dyn_cape_05_mid_l, bone2Name=dyn_cape_04_mid_l, swing1LimitDegrees=45, swing2LimitDegrees=45, twistLimitDegrees=30}
 {bone1Name=dyn_cape_06_mid_l, bone2Name=dyn_cape_05_mid_l, swing1LimitDegrees=45, swing2LimitDegrees=45, twistLimitDegrees=30}
 {bone1Name=dyn_cape_07_mid_l, bone2Name=dyn_cape_06_mid_l, swing1LimitDegrees=45, swing2LimitDegrees=45, twistLimitDegrees=30}
 {bone1Name=dyn_cape_08_mid_l, bone2Name=dyn_cape_07_mid_l, swing1LimitDegrees=45, swing2LimitDegrees=45, twistLimitDegrees=30}
 {bone1Name=dyn_cape_09_mid_l, bone2Name=dyn_cape_08_mid_l, swing1LimitDegrees=45, swing2LimitDegrees=45, twistLimitDegrees=30}
 {bone1Name=dyn_cape_010_mid_l, bone2Name=dyn_cape_09_mid_l, swing1LimitDegrees=45, swing2LimitDegrees=45, twistLimitDegrees=30}
 {bone1Name=dyn_cape_011_mid_l, bone2Name=dyn_cape_010_mid_l, swing1LimitDegrees=45, swing2LimitDegrees=45, twistLimitDegrees=30}
 {bone1Name=dyn_cape_011_l, bone2Name=dyn_cape_010_l, swing1LimitDegrees=45, swing2LimitDegrees=45, twistLimitDegrees=30}
 {
  bone1Name=dyn_cape_02_mid_l
  bone2Name=dyn_cape_02_l
  swing1LimitDegrees=45
  swing1Motion=Free
  swing2LimitDegrees=45
  swing2Motion=Free
  twistLimitDegrees=45
  twistMotion=Free
 }
 {
  bone1Name=dyn_cape_03_mid_l
  bone2Name=dyn_cape_03_l
  swing1LimitDegrees=45
  swing1Motion=Free
  swing2LimitDegrees=45
  swing2Motion=Free
  twistLimitDegrees=45
  twistMotion=Free
 }
 {
  bone1Name=dyn_cape_04_mid_l
  bone2Name=dyn_cape_04_l
  swing1LimitDegrees=45
  swing1Motion=Free
  swing2LimitDegrees=45
  swing2Motion=Free
  twistLimitDegrees=45
  twistMotion=Free
 }
 {
  bone1Name=dyn_cape_05_mid_l
  bone2Name=dyn_cape_05_l
  swing1LimitDegrees=45
  swing1Motion=Free
  swing2LimitDegrees=45
  swing2Motion=Free
  twistLimitDegrees=45
  twistMotion=Free
 }
 {
  bone1Name=dyn_cape_06_mid_l
  bone2Name=dyn_cape_06_l
  swing1LimitDegrees=45
  swing1Motion=Free
  swing2LimitDegrees=45
  swing2Motion=Free
  twistLimitDegrees=45
  twistMotion=Free
 }
 {
  bone1Name=dyn_cape_07_mid_l
  bone2Name=dyn_cape_07_l
  swing1LimitDegrees=45
  swing1Motion=Free
  swing2LimitDegrees=45
  swing2Motion=Free
  twistLimitDegrees=45
  twistMotion=Free
 }
 {
  bone1Name=dyn_cape_08_mid_l
  bone2Name=dyn_cape_08_l
  swing1LimitDegrees=45
  swing1Motion=Free
  swing2LimitDegrees=45
  swing2Motion=Free
  twistLimitDegrees=45
  twistMotion=Free
 }
 {
  bone1Name=dyn_cape_09_mid_l
  bone2Name=dyn_cape_09_l
  swing1LimitDegrees=45
  swing1Motion=Free
  swing2LimitDegrees=45
  swing2Motion=Free
  twistLimitDegrees=45
  twistMotion=Free
 }
 {
  bone1Name=dyn_cape_010_mid_l
  bone2Name=dyn_cape_010_l
  swing1LimitDegrees=45
  swing1Motion=Free
  swing2LimitDegrees=45
  swing2Motion=Free
  twistLimitDegrees=45
  twistMotion=Free
 }
 {
  bone1Name=dyn_cape_011_mid_l
  bone2Name=dyn_cape_011_l
  swing1LimitDegrees=45
  swing1Motion=Free
  swing2LimitDegrees=45
  swing2Motion=Free
  twistLimitDegrees=45
  twistMotion=Free
 }
 {
  bone1Name=dyn_cape_01_mid_r
  bone2Name=dyn_cape_01_r
  swing1LimitDegrees=45
  swing1Motion=Free
  swing2LimitDegrees=45
  swing2Motion=Free
  twistLimitDegrees=45
  twistMotion=Free
 }
 {
  bone1Name=dyn_cape_02_mid_r
  bone2Name=dyn_cape_02_r
  swing1LimitDegrees=45
  swing1Motion=Free
  swing2LimitDegrees=45
  swing2Motion=Free
  twistLimitDegrees=45
  twistMotion=Free
 }
 {
  bone1Name=dyn_cape_03_mid_r
  bone2Name=dyn_cape_03_r
  swing1LimitDegrees=45
  swing1Motion=Free
  swing2LimitDegrees=45
  swing2Motion=Free
  twistLimitDegrees=45
  twistMotion=Free
 }
 {
  bone1Name=dyn_cape_04_mid_r
  bone2Name=dyn_cape_04_r
  swing1LimitDegrees=45
  swing1Motion=Free
  swing2LimitDegrees=45
  swing2Motion=Free
  twistLimitDegrees=45
  twistMotion=Free
 }
 {
  bone1Name=dyn_cape_05_mid_r
  bone2Name=dyn_cape_05_r
  swing1LimitDegrees=45
  swing1Motion=Free
  swing2LimitDegrees=45
  swing2Motion=Free
  twistLimitDegrees=45
  twistMotion=Free
 }
 {
  bone1Name=dyn_cape_06_mid_r
  bone2Name=dyn_cape_06_r
  swing1LimitDegrees=45
  swing1Motion=Free
  swing2LimitDegrees=45
  swing2Motion=Free
  twistLimitDegrees=45
  twistMotion=Free
 }
 {
  bone1Name=dyn_cape_07_mid_r
  bone2Name=dyn_cape_07_r
  swing1LimitDegrees=45
  swing1Motion=Free
  swing2LimitDegrees=45
  swing2Motion=Free
  twistLimitDegrees=45
  twistMotion=Free
 }
 {
  bone1Name=dyn_cape_08_mid_r
  bone2Name=dyn_cape_08_r
  swing1LimitDegrees=45
  swing1Motion=Free
  swing2LimitDegrees=45
  swing2Motion=Free
  twistLimitDegrees=45
  twistMotion=Free
 }
 {
  bone1Name=dyn_cape_09_mid_r
  bone2Name=dyn_cape_09_r
  swing1LimitDegrees=45
  swing1Motion=Free
  swing2LimitDegrees=45
  swing2Motion=Free
  twistLimitDegrees=45
  twistMotion=Free
 }
 {
  bone1Name=dyn_cape_010_mid_r
  bone2Name=dyn_cape_010_r
  swing1LimitDegrees=45
  swing1Motion=Free
  swing2LimitDegrees=45
  swing2Motion=Free
  twistLimitDegrees=45
  twistMotion=Free
 }
 {
  bone1Name=dyn_cape_01_mid_l
  bone2Name=dyn_cape_01
  swing1LimitDegrees=45
  swing1Motion=Free
  swing2LimitDegrees=45
  swing2Motion=Free
  twistLimitDegrees=45
  twistMotion=Free
 }
 {
  bone1Name=dyn_cape_02_mid_l
  bone2Name=dyn_cape_02
  swing1LimitDegrees=45
  swing1Motion=Free
  swing2LimitDegrees=45
  swing2Motion=Free
  twistLimitDegrees=45
  twistMotion=Free
 }
 {
  bone1Name=dyn_cape_03_mid_l
  bone2Name=dyn_cape_03
  swing1LimitDegrees=45
  swing1Motion=Free
  swing2LimitDegrees=45
  swing2Motion=Free
  twistLimitDegrees=45
  twistMotion=Free
 }
 {
  bone1Name=dyn_cape_04_mid_l
  bone2Name=dyn_cape_04
  swing1LimitDegrees=45
  swing1Motion=Free
  swing2LimitDegrees=45
  swing2Motion=Free
  twistLimitDegrees=45
  twistMotion=Free
 }
 {
  bone1Name=dyn_cape_05_mid_l
  bone2Name=dyn_cape_05
  swing1LimitDegrees=45
  swing1Motion=Free
  swing2LimitDegrees=45
  swing2Motion=Free
  twistLimitDegrees=45
  twistMotion=Free
 }
 {
  bone1Name=dyn_cape_06_mid_l
  bone2Name=dyn_cape_06
  swing1LimitDegrees=45
  swing1Motion=Free
  swing2LimitDegrees=45
  swing2Motion=Free
  twistLimitDegrees=45
  twistMotion=Free
 }
 {
  bone1Name=dyn_cape_07_mid_l
  bone2Name=dyn_cape_07
  swing1LimitDegrees=45
  swing1Motion=Free
  swing2LimitDegrees=45
  swing2Motion=Free
  twistLimitDegrees=45
  twistMotion=Free
 }
 {
  bone1Name=dyn_cape_08_mid_l
  bone2Name=dyn_cape_08
  swing1LimitDegrees=45
  swing1Motion=Free
  swing2LimitDegrees=45
  swing2Motion=Free
  twistLimitDegrees=45
  twistMotion=Free
 }
 {
  bone1Name=dyn_cape_09_mid_l
  bone2Name=dyn_cape_09
  swing1LimitDegrees=45
  swing1Motion=Free
  swing2LimitDegrees=45
  swing2Motion=Free
  twistLimitDegrees=45
  twistMotion=Free
 }
 {
  bone1Name=dyn_cape_010_mid_l
  bone2Name=dyn_cape_010
  swing1LimitDegrees=45
  swing1Motion=Free
  swing2LimitDegrees=45
  swing2Motion=Free
  twistLimitDegrees=45
  twistMotion=Free
 }
 {
  bone1Name=dyn_cape_011_mid_l
  bone2Name=dyn_cape_011
  swing1LimitDegrees=45
  swing1Motion=Free
  swing2LimitDegrees=45
  swing2Motion=Free
  twistLimitDegrees=45
  twistMotion=Free
 }
 {
  bone1Name=dyn_cape_01_mid_r
  bone2Name=dyn_cape_01
  swing1LimitDegrees=45
  swing1Motion=Free
  swing2LimitDegrees=45
  swing2Motion=Free
  twistLimitDegrees=45
  twistMotion=Free
 }
 {
  bone1Name=dyn_cape_02_mid_r
  bone2Name=dyn_cape_02
  swing1LimitDegrees=45
  swing1Motion=Free
  swing2LimitDegrees=45
  swing2Motion=Free
  twistLimitDegrees=45
  twistMotion=Free
 }
 {
  bone1Name=dyn_cape_03_mid_r
  bone2Name=dyn_cape_03
  swing1LimitDegrees=45
  swing1Motion=Free
  swing2LimitDegrees=45
  swing2Motion=Free
  twistLimitDegrees=45
  twistMotion=Free
 }
 {
  bone1Name=dyn_cape_04_mid_r
  bone2Name=dyn_cape_04
  swing1LimitDegrees=45
  swing1Motion=Free
  swing2LimitDegrees=45
  swing2Motion=Free
  twistLimitDegrees=45
  twistMotion=Free
 }
 {
  bone1Name=dyn_cape_05_mid_r
  bone2Name=dyn_cape_05
  swing1LimitDegrees=45
  swing1Motion=Free
  swing2LimitDegrees=45
  swing2Motion=Free
  twistLimitDegrees=45
  twistMotion=Free
 }
 {
  bone1Name=dyn_cape_06_mid_r
  bone2Name=dyn_cape_06
  swing1LimitDegrees=45
  swing1Motion=Free
  swing2LimitDegrees=45
  swing2Motion=Free
  twistLimitDegrees=45
  twistMotion=Free
 }
 {
  bone1Name=dyn_cape_07_mid_r
  bone2Name=dyn_cape_07
  swing1LimitDegrees=45
  swing1Motion=Free
  swing2LimitDegrees=45
  swing2Motion=Free
  twistLimitDegrees=45
  twistMotion=Free
 }
 {
  bone1Name=dyn_cape_08_mid_r
  bone2Name=dyn_cape_08
  swing1LimitDegrees=45
  swing1Motion=Free
  swing2LimitDegrees=45
  swing2Motion=Free
  twistLimitDegrees=45
  twistMotion=Free
 }
 {
  bone1Name=dyn_cape_09_mid_r
  bone2Name=dyn_cape_09
  swing1LimitDegrees=45
  swing1Motion=Free
  swing2LimitDegrees=45
  swing2Motion=Free
  twistLimitDegrees=45
  twistMotion=Free
 }
 {
  bone1Name=dyn_cape_010_mid_r
  bone2Name=dyn_cape_010
  swing1LimitDegrees=45
  swing1Motion=Free
  swing2LimitDegrees=45
  swing2Motion=Free
  twistLimitDegrees=45
  twistMotion=Free
 }
 {
  bone1Name=pelvis
  bone2Name=root
  swing1LimitDegrees=45
  swing1Motion=Free
  swing2LimitDegrees=45
  swing2Motion=Free
  twistLimitDegrees=45
  twistMotion=Free
 }
]
