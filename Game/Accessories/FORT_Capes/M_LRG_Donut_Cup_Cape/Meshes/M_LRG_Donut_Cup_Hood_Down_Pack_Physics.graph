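# Physics asset for the hood-down pack mesh of the Donut Cup cape, setting up physics for the lowered hood.
# Capsules on spine_05 and both clavicles form the upper-body volume, with tiny MSTR anchor bodies constrained to spine_05 with free swing and twist.
# Eight hood capsules around the neck each constrain to their own anchor with tight angular limits, wider at front and back center.
# The five cape roots have only anchor bodies, with no downstream cape chain.
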PhysicsAsset /Game/Accessories/FORT_Capes/M_LRG_Donut_Cup_Cape/Meshes/M_LRG_Donut_Cup_Hood_Down_Pack_Physics
bodies {
 clavicle_l [{shapeType=Capsule, length=17.5, radius=11.25, center=<9.2074 -1.6511 -1.7917>, rotation=<rot -90 0 0.0001>}]
 clavicle_r [{shapeType=Capsule, length=17.5, radius=11.25, center=<-9.2074 1.6511 1.7917>, rotation=<rot -90 0 -179.9999>}]
 dyn_cape_01_MSTR [{shapeType=Capsule, length=0.01, radius=0.01}]
 dyn_cape_01_l_MSTR [{shapeType=Capsule, length=0.01, radius=0.01}]
 dyn_cape_01_mid_l_MSTR [{shapeType=Capsule, length=0.01, radius=0.01}]
 dyn_cape_01_mid_r_MSTR [{shapeType=Capsule, length=0.01, radius=0.01}]
 dyn_cape_01_r_MSTR [{shapeType=Capsule, length=0.01, radius=0.01}]
 dyn_hooddown_01_l [{shapeType=Capsule, length=3.9417, radius=5.142, center=<2.9507 -0.0002 0.0001>, rotation=<rot 90 -0.0049 -0.0049>}]
 dyn_hooddown_01_l_MSTR [{shapeType=Box, extentX=0.01, extentY=0.01, extentZ=0.01}]
 dyn_hooddown_01_r [{shapeType=Capsule, length=3.9417, radius=5.142, center=<2.9507 -0.0002 0.0001>, rotation=<rot 90 -0.0049 -0.0049>}]
 dyn_hooddown_01_r_MSTR [{shapeType=Box, extentX=0.01, extentY=0.01, extentZ=0.01}]
 dyn_hooddown_bk_01 [{shapeType=Capsule, length=3.9417, radius=5.142, center=<3.3058 -0.0002 0>, rotation=<rot 90 0.0198 0.0198>}]
 dyn_hooddown_bk_01_MSTR [{shapeType=Box, extentX=0.01, extentY=0.01, extentZ=0.01}]
 dyn_hooddown_bk_01_l [{shapeType=Capsule, length=3.9417, radius=5.142, center=<2.9507 -0.0002 0.0001>, rotation=<rot 90 -0.0049 -0.0049>}]
 dyn_hooddown_bk_01_l_MSTR [{shapeType=Box, extentX=0.01, extentY=0.01, extentZ=0.01}]
 dyn_hooddown_bk_01_r [{shapeType=Capsule, length=3.9417, radius=5.142, center=<2.9507 -0.0002 0.0001>, rotation=<rot 90 -0.0049 -0.0049>}]
 dyn_hooddown_bk_01_r_MSTR [{shapeType=Box, extentX=0.01, extentY=0.01, extentZ=0.01}]
 dyn_hooddown_fr_01 [{shapeType=Capsule, length=3.9417, radius=5.142, center=<1.2375 -0.0002 0.0001>, rotation=<rot 90 -0.0198 -0.0198>}]
 dyn_hooddown_fr_01_MSTR [{shapeType=Box, extentX=0.01, extentY=0.01, extentZ=0.01}]
 dyn_hooddown_fr_01_l [{shapeType=Capsule, length=3.9417, radius=5.142, center=<2.9507 -0.0002 0.0001>, rotation=<rot 90 -0.0049 -0.0049>}]
 dyn_hooddown_fr_01_l_MSTR [{shapeType=Box, extentX=0.01, extentY=0.01, extentZ=0.01}]
 dyn_hooddown_fr_01_r [{shapeType=Capsule, length=3.9417, radius=5.142, center=<2.9507 -0.0002 0.0001>, rotation=<rot 90 -0.0049 -0.0049>}]
 dyn_hooddown_fr_01_r_MSTR [{shapeType=Box, extentX=0.01, extentY=0.01, extentZ=0.01}]
 spine_05 [{shapeType=Capsule, length=13, radius=15.75, center=<-9.037 -3.7779 0>}]
}
constraints [
 {
  bone1Name=dyn_cape_01_l_MSTR
  bone2Name=spine_05
  swing1LimitDegrees=45
  swing1Motion=Free
  swing2LimitDegrees=45
  swing2Motion=Free
  twistLimitDegrees=45
  twistMotion=Free
 }
 {
  bone1Name=clavicle_l
  bone2Name=spine_05
  swing1LimitDegrees=45
  swing1Motion=Free
  swing2LimitDegrees=45
  swing2Motion=Free
  twistLimitDegrees=45
  twistMotion=Free
 }
 {
  bone1Name=clavicle_r
  bone2Name=spine_05
  swing1LimitDegrees=45
  swing1Motion=Free
  swing2LimitDegrees=45
  swing2Motion=Free
  twistLimitDegrees=45
  twistMotion=Free
 }
 {
  bone1Name=dyn_cape_01_r_MSTR
  bone2Name=spine_05
  swing1LimitDegrees=45
  swing1Motion=Free
  swing2LimitDegrees=45
  swing2Motion=Free
  twistLimitDegrees=45
  twistMotion=Free
 }
 {
  bone1Name=dyn_cape_01_mid_r_MSTR
  bone2Name=spine_05
  swing1LimitDegrees=45
  swing1Motion=Free
  swing2LimitDegrees=45
  swing2Motion=Free
  twistLimitDegrees=45
  twistMotion=Free
 }
 {
  bone1Name=dyn_cape_01_MSTR
  bone2Name=spine_05
  swing1LimitDegrees=45
  swing1Motion=Free
  swing2LimitDegrees=45
  swing2Motion=Free
  twistLimitDegrees=45
  twistMotion=Free
 }
 {
  bone1Name=dyn_cape_01_mid_l_MSTR
  bone2Name=spine_05
  swing1LimitDegrees=45
  swing1Motion=Free
  swing2LimitDegrees=45
  swing2Motion=Free
  twistLimitDegrees=45
  twistMotion=Free
 }
 {
  bone1Name=dyn_hooddown_bk_01_MSTR
  bone2Name=spine_05
  swing1LimitDegrees=45
  swing1Motion=Free
  swing2LimitDegrees=45
  swing2Motion=Free
  twistLimitDegrees=45
  twistMotion=Free
 }
 {
  bone1Name=dyn_hooddown_01_l_MSTR
  bone2Name=spine_05
  swing1LimitDegrees=45
  swing1Motion=Free
  swing2LimitDegrees=45
  swing2Motion=Free
  twistLimitDegrees=45
  twistMotion=Free
 }
 {
  bone1Name=dyn_hooddown_bk_01_l_MSTR
  bone2Name=spine_05
  swing1LimitDegrees=45
  swing1Motion=Free
  swing2LimitDegrees=45
  swing2Motion=Free
  twistLimitDegrees=45
  twistMotion=Free
 }
 {
  bone1Name=dyn_hooddown_fr_01_l_MSTR
  bone2Name=spine_05
  swing1LimitDegrees=45
  swing1Motion=Free
  swing2LimitDegrees=45
  swing2Motion=Free
  twistLimitDegrees=45
  twistMotion=Free
 }
 {
  bone1Name=dyn_hooddown_fr_01_MSTR
  bone2Name=spine_05
  swing1LimitDegrees=45
  swing1Motion=Free
  swing2LimitDegrees=45
  swing2Motion=Free
  twistLimitDegrees=45
  twistMotion=Free
 }
 {
  bone1Name=dyn_hooddown_fr_01_r_MSTR
  bone2Name=spine_05
  swing1LimitDegrees=45
  swing1Motion=Free
  swing2LimitDegrees=45
  swing2Motion=Free
  twistLimitDegrees=45
  twistMotion=Free
 }
 {
  bone1Name=dyn_hooddown_bk_01_r_MSTR
  bone2Name=spine_05
  swing1LimitDegrees=45
  swing1Motion=Free
  swing2LimitDegrees=45
  swing2Motion=Free
  twistLimitDegrees=45
  twistMotion=Free
 }
 {
  bone1Name=dyn_hooddown_01_r_MSTR
  bone2Name=spine_05
  swing1LimitDegrees=45
  swing1Motion=Free
  swing2LimitDegrees=45
  swing2Motion=Free
  twistLimitDegrees=45
  twistMotion=Free
 }
 {bone1Name=dyn_hooddown_bk_01, bone2Name=dyn_hooddown_bk_01_MSTR, swing1LimitDegrees=45, swing2LimitDegrees=45, twistLimitDegrees=5}
 {bone1Name=dyn_hooddown_01_l, bone2Name=dyn_hooddown_01_l_MSTR, swing1LimitDegrees=5, swing2LimitDegrees=20, twistLimitDegrees=5}
 {bone1Name=dyn_hooddown_bk_01_l, bone2Name=dyn_hooddown_bk_01_l_MSTR, swing1LimitDegrees=5, swing2LimitDegrees=20, twistLimitDegrees=5}
 {bone1Name=dyn_hooddown_fr_01_l, bone2Name=dyn_hooddown_fr_01_l_MSTR, swing1LimitDegrees=5, swing2LimitDegrees=20, twistLimitDegrees=5}
 {bone1Name=dyn_hooddown_fr_01, bone2Name=dyn_hooddown_fr_01_MSTR, swing1LimitDegrees=5, swing2LimitDegrees=45, twistLimitDegrees=5}
 {bone1Name=dyn_hooddown_fr_01_r, bone2Name=dyn_hooddown_fr_01_r_MSTR, swing1LimitDegrees=5, swing2LimitDegrees=20, twistLimitDegrees=5}
 {bone1Name=dyn_hooddown_bk_01_r, bone2Name=dyn_hooddown_bk_01_r_MSTR, swing1LimitDegrees=5, swing2LimitDegrees=20, twistLimitDegrees=5}
 {bone1Name=dyn_hooddown_01_r, bone2Name=dyn_hooddown_01_r_MSTR, swing1LimitDegrees=5, swing2LimitDegrees=20, twistLimitDegrees=5}
]
