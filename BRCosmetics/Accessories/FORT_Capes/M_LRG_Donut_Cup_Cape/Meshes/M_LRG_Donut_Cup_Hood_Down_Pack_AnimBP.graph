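# Post-process style AnimBlueprint that drives secondary motion for the Donut Cup cape and lowered hood cosmetic.
# Trail controllers on the copied parent pose feed clavicle-constrained cape and hood RigidBody sims, using scaled skydiving force as gravity.
# Output falls back to trails alone when rigid bodies are disabled, LOD exceeds 2, or emotes/custom animations apply.
# Bone offsets follow, including shoulder-twist hood correction; skydiving zeroes cloth max distance.
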AnimBlueprint /BRCosmetics/Accessories/FORT_Capes/M_LRG_Donut_Cup_Cape/Meshes/M_LRG_Donut_Cup_Hood_Down_Pack_AnimBP
target_skeleton=/Game/Accessories/FORT_Capes/M_LRG_Donut_Cup_Cape/Meshes/M_LRG_Donut_Cup_Hood_Down_Pack_Skeleton
graphs [
 AnimGraph {nodes [
   n1 = "Constraint - dyn_cape_01_r_MSTR (P)"(ComponentPose=Local To Component(LocalPose=Use cached pose 'trailControlCache'()), Alpha=1, ConstraintWeights_0=1) {
    alphaBoolBlend {blendOption=Linear}
    boneToModify {boneName=dyn_cape_01_r_MSTR}
    constraintSetup [{targetBone {boneName=clavicle_r}}]
    constraintWeights [1]
   }
   n2 = "Transform (Modify) Bone - Bone: dyn_cape_01_r_MSTR"(ComponentPose=@n1, Alpha=1, Translation="0.000,-0.000,1.856", Rotation="0,0,0", Scale="1,1,1") {
    rotationMode=BMM_Additive
    rotationSpace=BCS_BoneSpace
    translationMode=BMM_Additive
    translationSpace=BCS_BoneSpace
    alphaBoolBlend {blendOption=Linear}
    boneToModify {boneName=dyn_cape_01_r_MSTR}
    translation {z=1.856}
   }
   n3 = "Constraint - dyn_cape_01_l_MSTR (P)"(ComponentPose=@n2, Alpha=1, ConstraintWeights_0=1) {
    alphaBoolBlend {blendOption=Linear}
    boneToModify {boneName=dyn_cape_01_l_MSTR}
    constraintSetup [{targetBone {boneName=clavicle_l}}]
    constraintWeights [1]
   }
   n4 = Get windVector() -> (windVector:Vector)
   n5 = RigidBody!(ComponentPose=@n3, Alpha, OverrideWorldGravity:Vector=@n4, ExternalForce) -> (Pose) {
    bClampLinearTranslationLimitToRefPose=true
    bOverrideWorldGravity=true
    overridePhysicsAsset=/Game/Accessories/FORT_Capes/M_LRG_Donut_Cup_Cape/Meshes/M_LRG_Donut_Cup_Cape_Pack_Physics
    simulationSpace=BaseBoneSpace
    alphaBoolBlend {blendOption=Linear}
    baseBoneRef {boneName=root}
    componentAppliedLinearAccClamp=<3000 3000 3000>
    componentLinearAccScale=<0.1 0.1 0.4>
    componentLinearVelScale=<0.4 0.4 0.7>
    overrideWorldGravity {z=-980}
   }
   n6 = "Constraint - dyn_hooddown_bk_01_MSTR (R)"(ComponentPose=@n5, Alpha=1, ConstraintWeights_0=1) {
    alphaBoolBlend {blendOption=Linear}
    boneToModify {boneName=dyn_hooddown_bk_01_MSTR}
    constraintSetup [{transformType=Rotation, targetBone {boneName=dyn_cape_01}}]
    constraintWeights [1]
   }
   n7 = RigidBody(ComponentPose=@n6, OverrideWorldGravity=@n4) {
    bClampLinearTranslationLimitToRefPose=true
    bOverrideWorldGravity=true
    overridePhysicsAsset=/Game/Accessories/FORT_Capes/M_LRG_Donut_Cup_Cape/Meshes/M_LRG_Donut_Cup_Hood_Down_Pack_Physics
    simulationSpace=BaseBoneSpace
    alphaBoolBlend {blendOption=Linear}
    baseBoneRef {boneName=root}
    componentAppliedLinearAccClamp=<3000 3000 3000>
    componentLinearAccScale=<0.4 0.3 0.3>
    componentLinearVelScale=<0.4 0.3 0.5>
    overrideWorldGravity {z=-980}
   }
   n8 = Component To Local(ComponentPose=@n7)
   n9 = Use cached pose 'trailControlCache'()
   n10 = Get rigidsOnOffBool() -> (rigidsOnOffBool:bool)
   n11 = "Transform (Modify) Bone - Bone: dyn_cape_01_MSTR"(
    ComponentPose=Local To Component(LocalPose=Blend Poses by bool(BlendPose_0=@n9, BlendPose_1=@n8, bActiveValue=@n10) {blendType=Linear})
    Alpha=1
    Translation="0.250,-4,0.000"
    Rotation="0,0,0"
    Scale="1,1,1"
   ) {
    translationMode=BMM_Additive
    translationSpace=BCS_ParentBoneSpace
    alphaBoolBlend {blendOption=Linear}
    boneToModify {boneName=dyn_cape_01_MSTR}
    translation=<0.25 -4>
   }
   n12 = "Transform (Modify) Bone - Bone: dyn_cape_01_mid_r_MSTR"(ComponentPose=@n11, Alpha=1, Translation="0,-3,0.000", Rotation="0,0,0", Scale="1,1,1") {
    translationMode=BMM_Additive
    translationSpace=BCS_ParentBoneSpace
    alphaBoolBlend {blendOption=Linear}
    boneToModify {boneName=dyn_cape_01_mid_r_MSTR}
    translation {y=-3}
   }
   n13 = "Transform (Modify) Bone - Bone: dyn_cape_01_mid_l_MSTR"(ComponentPose=@n12, Alpha=1, Translation="0,-3,0.000", Rotation="0,0,0", Scale="1,1,1") {
    translationMode=BMM_Additive
    translationSpace=BCS_ParentBoneSpace
    alphaBoolBlend {blendOption=Linear}
    boneToModify {boneName=dyn_cape_01_mid_l_MSTR}
    translation {y=-3}
   }
   n14 = "Transform (Modify) Bone - Bone: dyn_hooddown_01_r_MSTR"(
    ComponentPose=@n13
    Alpha=Get RshoulderRotateBackAlpha() -> (RshoulderRotateBackAlpha:double)
    Translation="-0.620,0.000,3.510"
    Rotation="10.000,0.000,0.000"
    Scale="1,1,1"
   ) {
    rotationMode=BMM_Additive
    rotationSpace=BCS_BoneSpace
    translationMode=BMM_Additive
    translationSpace=BCS_BoneSpace
    alphaBoolBlend {blendOption=Linear}
    boneToModify {boneName=dyn_hooddown_01_r_MSTR}
    rotation {pitch=10}
    translation=<-0.6201 0 3.5104>
   }
   Output Pose(Result=Component To Local(ComponentPose=@n14)) {layerGroup=None}
   Save cached pose 'initCache'(Pose=Copy Pose From Mesh() {bUseAttachedParent=true}) {cacheName=initCache}
   n15 = Trail controller - Bone: dyn_cape_011_l!(
    ComponentPose=Local To Component(LocalPose=Use cached pose 'initCache'())
    Alpha=1
    RelaxationSpeedScale:float=1.5
    FakeVelocity:Vector="0,0,0"
   ) -> (Pose) {
    bActorSpaceFakeVel=true
    bShowTrailLocation=true
    bUsePlanarLimit=true
    chainLength=9
    debugLifeTime=0.1
    maxDeltaTime=0.0333
    relaxationSpeedScale=1.2
    alphaScaleBiasClamp {bInterpResult=true, interpSpeedDecreasing=25, interpSpeedIncreasing=25}
    baseJoint {boneName=root}
    planarLimits [
     {drivingBone {boneName=root}, planeTransform {location=<0 0 10>, rotation=<rot 0 0 0>, scale=<1 1 1>}}
     {drivingBone {boneName=spine_03}, planeTransform {location=<0 -20 0>, rotation=<rot 0 15 -89.9999>, scale=<1 1 1>}}
    ]
    rotationLimits [
     {limitMax=<180 180 180>}
     {limitMax=<180 180 180>}
     {limitMax=<180 180 180>}
     {limitMax=<180 180 180>}
     {limitMax=<180 180 180>}
     {limitMax=<180 180 180>}
     {limitMax=<180 180 180>}
     {limitMax=<180 180 180>}
     {limitMax=<180 180 180>}
    ]
    rotationOffsets [<0 0 0>, <0 0 0>, <0 0 0>, <0 0 0>, <0 0 0>, <0 0 0>, <0 0 0>, <0 0 0>, <0 0 0>]
    trailBone {boneName=dyn_cape_011_l}
    trailRelaxationSpeed {editorCurveData {keys [<at 0 15 RCIM_Linear>,<at 0.5289 13.1838 RCIM_Linear>,<at 1 6 RCIM_Linear>]}}
   }
   n16 = Trail controller - Bone: dyn_cape_011_mid_l!(ComponentPose=@n15, Alpha=1, RelaxationSpeedScale:float=1.5, FakeVelocity:Vector="0,0,0") -> (Pose) {
    bActorSpaceFakeVel=true
    bShowTrailLocation=true
    bUsePlanarLimit=true
    chainLength=9
    debugLifeTime=0.1
    maxDeltaTime=0.0333
    relaxationSpeedScale=1.2
    alphaScaleBiasClamp {bInterpResult=true, interpSpeedDecreasing=25, interpSpeedIncreasing=25}
    baseJoint {boneName=root}
    planarLimits [
     {drivingBone {boneName=root}, planeTransform {location=<0 0 10>, rotation=<rot 0 0 0>, scale=<1 1 1>}}
     {drivingBone {boneName=spine_03}, planeTransform {location=<0 -20 0>, rotation=<rot 0 15 -89.9999>, scale=<1 1 1>}}
    ]
    rotationLimits [
     {limitMax=<180 180 180>}
     {limitMax=<180 180 180>}
     {limitMax=<180 180 180>}
     {limitMax=<180 180 180>}
     {limitMax=<180 180 180>}
     {limitMax=<180 180 180>}
     {limitMax=<180 180 180>}
     {limitMax=<180 180 180>}
     {limitMax=<180 180 180>}
    ]
    rotationOffsets [<0 0 0>, <0 0 0>, <0 0 0>, <0 0 0>, <0 0 0>, <0 0 0>, <0 0 0>, <0 0 0>, <0 0 0>]
    trailBone {boneName=dyn_cape_011_mid_l}
    trailRelaxationSpeed {editorCurveData {keys [<at 0 15 RCIM_Linear>,<at 0.5289 13.1838 RCIM_Linear>,<at 1 6 RCIM_Linear>]}}
   }
   n17 = Trail controller - Bone: dyn_cape_011!(ComponentPose=@n16, Alpha=1, RelaxationSpeedScale:float=1.5, FakeVelocity:Vector="0,0,0") -> (Pose) {
    bActorSpaceFakeVel=true
    bShowTrailLocation=true
    bUsePlanarLimit=true
    chainLength=9
    debugLifeTime=0.1
    maxDeltaTime=0.0333
    relaxationSpeedScale=1.2
    alphaScaleBiasClamp {bInterpResult=true, interpSpeedDecreasing=25, interpSpeedIncreasing=25}
    baseJoint {boneName=root}
    planarLimits [
     {drivingBone {boneName=root}, planeTransform {location=<0 0 10>, rotation=<rot 0 0 0>, scale=<1 1 1>}}
     {drivingBone {boneName=spine_03}, planeTransform {location=<0 -20 0>, rotation=<rot 0 15 -89.9999>, scale=<1 1 1>}}
    ]
    rotationLimits [
     {limitMax=<180 180 180>}
     {limitMax=<180 180 180>}
     {limitMax=<180 180 180>}
     {limitMax=<180 180 180>}
     {limitMax=<180 180 180>}
     {limitMax=<180 180 180>}
     {limitMax=<180 180 180>}
     {limitMax=<180 180 180>}
     {limitMax=<180 180 180>}
    ]
    rotationOffsets [<0 0 0>, <0 0 0>, <0 0 0>, <0 0 0>, <0 0 0>, <0 0 0>, <0 0 0>, <0 0 0>, <0 0 0>]
    trailBone {boneName=dyn_cape_011}
    trailRelaxationSpeed {editorCurveData {keys [<at 0 15 RCIM_Linear>,<at 0.5289 13.1838 RCIM_Linear>,<at 1 6 RCIM_Linear>]}}
   }
   n18 = Trail controller - Bone: dyn_cape_010_mid_r!(ComponentPose=@n17, Alpha=1, RelaxationSpeedScale:float=1.5, FakeVelocity:Vector="0,0,0") -> (Pose) {
    bActorSpaceFakeVel=true
    bShowTrailLocation=true
    bUsePlanarLimit=true
    chainLength=9
    debugLifeTime=0.1
    maxDeltaTime=0.0333
    relaxationSpeedScale=1.2
    alphaScaleBiasClamp {bInterpResult=true, interpSpeedDecreasing=25, interpSpeedIncreasing=25}
    baseJoint {boneName=root}
    planarLimits [
     {drivingBone {boneName=root}, planeTransform {location=<0 0 10>, rotation=<rot 0 0 0>, scale=<1 1 1>}}
     {drivingBone {boneName=spine_03}, planeTransform {location=<0 -20 0>, rotation=<rot 0 15 -89.9999>, scale=<1 1 1>}}
    ]
    rotationLimits [
     {limitMax=<180 180 180>}
     {limitMax=<180 180 180>}
     {limitMax=<180 180 180>}
     {limitMax=<180 180 180>}
     {limitMax=<180 180 180>}
     {limitMax=<180 180 180>}
     {limitMax=<180 180 180>}
     {limitMax=<180 180 180>}
     {limitMax=<180 180 180>}
    ]
    rotationOffsets [<0 0 0>, <0 0 0>, <0 0 0>, <0 0 0>, <0 0 0>, <0 0 0>, <0 0 0>, <0 0 0>, <0 0 0>]
    trailBone {boneName=dyn_cape_010_mid_r}
    trailRelaxationSpeed {editorCurveData {keys [<at 0 15 RCIM_Linear>,<at 0.5289 13.1838 RCIM_Linear>,<at 1 6 RCIM_Linear>]}}
   }
   n19 = Trail controller - Bone: dyn_cape_010_r!(ComponentPose=@n18, Alpha=1, RelaxationSpeedScale:float=1.5, FakeVelocity:Vector="0,0,0") -> (Pose) {
    bActorSpaceFakeVel=true
    bShowTrailLocation=true
    bUsePlanarLimit=true
    chainLength=9
    debugLifeTime=0.1
    maxDeltaTime=0.0333
    relaxationSpeedScale=1.2
    alphaScaleBiasClamp {bInterpResult=true, interpSpeedDecreasing=25, interpSpeedIncreasing=25}
    baseJoint {boneName=root}
    planarLimits [
     {drivingBone {boneName=root}, planeTransform {location=<0 0 10>, rotation=<rot 0 0 0>, scale=<1 1 1>}}
     {drivingBone {boneName=spine_03}, planeTransform {location=<0 -20 0>, rotation=<rot 0 15 -89.9999>, scale=<1 1 1>}}
    ]
    rotationLimits [
     {limitMax=<180 180 180>}
     {limitMax=<180 180 180>}
     {limitMax=<180 180 180>}
     {limitMax=<180 180 180>}
     {limitMax=<180 180 180>}
     {limitMax=<180 180 180>}
     {limitMax=<180 180 180>}
     {limitMax=<180 180 180>}
     {limitMax=<180 180 180>}
    ]
    rotationOffsets [<0 0 0>, <0 0 0>, <0 0 0>, <0 0 0>, <0 0 0>, <0 0 0>, <0 0 0>, <0 0 0>, <0 0 0>]
    trailBone {boneName=dyn_cape_010_r}
    trailRelaxationSpeed {editorCurveData {keys [<at 0 15 RCIM_Linear>,<at 0.5289 13.1838 RCIM_Linear>,<at 1 6 RCIM_Linear>]}}
   }
   Save cached pose 'trailControlCache'(Pose=Component To Local(ComponentPose=@n19)) {cacheName=trailControlCache}
  ]}
 EventGraph {nodes [
   Event BlueprintUpdateAnimation()
   > n1 = Sequence!(execute) -> (then_0, then_1, then_2:Exec, then_3:Exec, then_4:Exec)
   > Set RshoulderRotateBackAlpha(RshoulderRotateBackAlpha:double=GetAlphaFromReferenceTransformDelta(
     SkeletalMeshComponent=GetPartSkeletalMeshComponent()
     BoneName=upperarm_r
     BaseName=clavicle_r
     DeltaComponentType=QuaternionTwist
     InputRange=(X=-55,Y=-125)
     OutputRange=(X=1,Y=0)
     TwistAxis="0,0,1"
    )) -> (Output_Get:double)
   n2 = Break Fort Anim Input Skydiving External Force(FortAnimInput_SkydivingExternalForce=Get SkydivingExternalForceInput()) -> (
    bUseSkydivingVectorForce=~
    SkydivingVectorForce=~
    HeadToPelvisDirection=~
    FloatingMultiplier=~
    FloatingAdditive=~
    DivingMultiplier=~
    DivingAdditive=~
    ParachutingMultiplier=~
    ParachutingAdditive=~
   )
   Set windVector(
    execute=@n1.then_1
    windVector:Vector=SelectVector(A=vector * float(A=@n2.SkydivingVectorForce, B=0.1), B="0, 0,-980", bPickA=Get bIsSkyDiving())
   ) -> (Output_Get:Vector)
   n3 = Get CurrentLOD()
   n4 = Get bRigidBodyEnabled()
   Set rigidsOnOffBool(
    execute=@n1.then_2
    rigidsOnOffBool:bool=OR Boolean!(
     A=NOT Boolean(A=@n4)
     B="integer > integer"(A=@n3, B=2) {is=title K2Node_CallFunction}
     C:bool=Get bCanPlayCustomAnimations() -> (bCanPlayCustomAnimations:bool)
     D:bool=Get bIsPlayingEmote()
    ) -> (ReturnValue)
   ) -> (Output_Get:bool)
   n5 = Branch(execute=@n1.then_3, Condition=Get bIsSkyDiving())
   n6 = GetPartSkeletalMeshComponent()
   > SetClothMaxDistanceScale!(execute, self:Skeletal Mesh Component Object Reference=@n6, Scale:float=0) -> (then)
   SetClothMaxDistanceScale(execute=@n5.else, self=@n6, Scale=3)
   TryGetPawnOwner()
  ]}
]
variables [
 rigidsOnOffBool {type=boolean}
 RshoulderRotateBackAlpha {type=number}
 windVector {type=Vector}
]
