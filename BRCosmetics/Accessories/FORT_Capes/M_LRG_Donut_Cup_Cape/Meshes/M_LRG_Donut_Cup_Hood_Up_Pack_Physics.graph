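# Physics asset for the hood-up pack of the Donut Cup cape cosmetic, defining physics bodies and joints for its hood bones.
# Tiny capsules on MSTR bones act as attachment points for larger capsules on back, side, lower and hood-down bones, joined by tight angular limits.
# Left and right hood chains add a second link, hood-down allows wider swing on one axis, and the top MSTR body has no constraint.
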PhysicsAsset /BRCosmetics/Accessories/FORT_Capes/M_LRG_Donut_Cup_Cape/Meshes/M_LRG_Donut_Cup_Hood_Up_Pack_Physics
bodies {
 dyn_hood_bk_01 [{shapeType=Capsule, length=6.1917, radius=2.892, center=<2.9507 -0.0002 0.0001>, rotation=<rot 90 -0.0049 -0.0049>}]
 dyn_hood_bk_MSTR [{shapeType=Capsule, length=0.01, radius=0.01}]
 dyn_hood_bk_l_01 [{shapeType=Capsule, length=6.1917, radius=2.892, center=<2.9507 -0.0002 0.0001>, rotation=<rot 90 -0.0049 -0.0049>}]
 dyn_hood_bk_l_MSTR [{shapeType=Capsule, length=0.01, radius=0.01}]
 dyn_hood_bk_r_01 [{shapeType=Capsule, length=6.1917, radius=2.892, center=<2.9507 -0.0002 0.0001>, rotation=<rot 90 -0.0049 -0.0049>}]
 dyn_hood_bk_r_MSTR [{shapeType=Capsule, length=0.01, radius=0.01}]
 dyn_hood_l_01 [{shapeType=Capsule, length=6.1917, radius=2.892, center=<2.9507 -0.0002 0.0001>, rotation=<rot 90 -0.0049 -0.0049>}]
 dyn_hood_l_02_a [{shapeType=Capsule, length=9.2426, radius=1.5529, rotation=<rot 0 0 -90>}]
 dyn_hood_l_MSTR [{shapeType=Capsule, length=0.01, radius=0.01}]
 dyn_hood_r_01 [{shapeType=Capsule, length=6.1917, radius=2.892, center=<2.9507 -0.0002 0.0001>, rotation=<rot 90 -0.0049 -0.0049>}]
 dyn_hood_r_02_a [{shapeType=Capsule, length=9.2426, radius=1.5529, rotation=<rot 0 0 -90>}]
 dyn_hood_r_MSTR [{shapeType=Capsule, length=0.01, radius=0.01}]
 dyn_hood_top_MSTR [{shapeType=Capsule, length=0.01, radius=0.01}]
 dyn_hooddown_bk_01 [{shapeType=Capsule, length=3.9417, radius=5.142, center=<3.3058 -0.0002 0>, rotation=<rot 90 0.0198 0.0198>}]
 dyn_hooddown_bk_01_MSTR [{shapeType=Capsule, length=0.01, radius=0.01}]
 dyn_hoodlower_bk_01_l [{shapeType=Capsule, length=3.9417, radius=5.142, center=<2.9507 -0.0002 0.0001>, rotation=<rot 90 -0.0049 -0.0049>}]
 dyn_hoodlower_bk_01_l_MSTR [{shapeType=Capsule, length=0.01, radius=0.01}]
 dyn_hoodlower_bk_01_r [{shapeType=Capsule, length=3.9417, radius=5.142, center=<2.9507 -0.0002 0.0001>, rotation=<rot 90 -0.0049 -0.0049>}]
 dyn_hoodlower_bk_01_r_MSTR [{shapeType=Capsule, length=0.01, radius=0.01}]
 dyn_hoodlower_fr_01_l [{shapeType=Capsule, length=3.9417, radius=5.142, center=<2.9507 -0.0002 0.0001>, rotation=<rot 90 -0.0049 -0.0049>}]
 dyn_hoodlower_fr_01_l_MSTR [{shapeType=Capsule, length=0.01, radius=0.01}]
 dyn_hoodlower_fr_01_r [{shapeType=Capsule, length=3.9417, radius=5.142, center=<2.9507 -0.0002 0.0001>, rotation=<rot 90 -0.0049 -0.0049>}]
 dyn_hoodlower_fr_01_r_MSTR [{shapeType=Capsule, length=0.01, radius=0.01}]
}
constraints [
 {bone1Name=dyn_hoodlower_fr_01_l, bone2Name=dyn_hoodlower_fr_01_l_MSTR, swing1LimitDegrees=15, swing2LimitDegrees=15, twistLimitDegrees=5}
 {bone1Name=dyn_hoodlower_bk_01_l, bone2Name=dyn_hoodlower_bk_01_l_MSTR, swing1LimitDegrees=15, swing2LimitDegrees=15, twistLimitDegrees=5}
 {bone1Name=dyn_hoodlower_bk_01_r, bone2Name=dyn_hoodlower_bk_01_r_MSTR, swing1LimitDegrees=15, swing2LimitDegrees=15, twistLimitDegrees=5}
 {bone1Name=dyn_hoodlower_fr_01_r, bone2Name=dyn_hoodlower_fr_01_r_MSTR, swing1LimitDegrees=15, swing2LimitDegrees=15, twistLimitDegrees=5}
 {bone1Name=dyn_hood_bk_01, bone2Name=dyn_hood_bk_MSTR, swing1LimitDegrees=15, swing2LimitDegrees=15, twistLimitDegrees=5}
 {bone1Name=dyn_hood_l_01, bone2Name=dyn_hood_l_MSTR, swing1LimitDegrees=15, swing2LimitDegrees=15, twistLimitDegrees=5}
 {bone1Name=dyn_hood_bk_l_01, bone2Name=dyn_hood_bk_l_MSTR, swing1LimitDegrees=15, swing2LimitDegrees=15, twistLimitDegrees=5}
 {bone1Name=dyn_hood_r_01, bone2Name=dyn_hood_r_MSTR, swing1LimitDegrees=15, swing2LimitDegrees=15, twistLimitDegrees=5}
 {bone1Name=dyn_hood_bk_r_01, bone2Name=dyn_hood_bk_r_MSTR, swing1LimitDegrees=15, swing2LimitDegrees=15, twistLimitDegrees=5}
 {bone1Name=dyn_hooddown_bk_01, bone2Name=dyn_hooddown_bk_01_MSTR, swing1LimitDegrees=30, swing2LimitDegrees=5, twistLimitDegrees=5}
 {bone1Name=dyn_hood_l_02_a, bone2Name=dyn_hood_l_01, swing1LimitDegrees=5, swing2LimitDegrees=15, twistLimitDegrees=5}
 {bone1Name=dyn_hood_r_02_a, bone2Name=dyn_hood_r_01, swing1LimitDegrees=5, swing2LimitDegrees=15, twistLimitDegrees=5}
]
