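; Cosmetic variant script for the DonutCup back bling (BID_506) that decides which variant tag applies based on the player's loadout.
; DetermineLoadoutVariantSelection reads the backpack's active Parts channel variant.
; When the matching DonutCup character (CID_738) is also equipped, Stage1 and Stage2 are remapped to Mat1 and Mat2.
; In all other cases it returns the backpack's Parts selection unchanged.
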(fn DetermineLoadoutVariantSelection
  (params (Loadout :type Cosmetic%20Loadout%20Structure%20%28by%20ref%29))
  (if (Class|FortCosmeticFunctionLibrary|IsItemInLoadout Loadout "/BRCosmetics/Athena/Items/Cosmetics/Characters/CID_738_Athena_Commando_M_DonutCup.CID_738_Athena_Commando_M_DonutCup")
    (if (== (Class|FortCosmeticFunctionLibrary|GetActiveVariantForSoftItemAndChannel Loadout "/BRCosmetics/Athena/Items/Cosmetics/Backpacks/BID_506_DonutCup.BID_506_DonutCup" "(TagName=\"Cosmetics.Variant.Channel.Parts\")") "(TagName=\"Cosmetics.Variant.Property.Stage1\")")
      (return "(TagName=\"Cosmetics.Variant.Property.Mat1\")")
      (elif (== (Class|FortCosmeticFunctionLibrary|GetActiveVariantForSoftItemAndChannel Loadout "/BRCosmetics/Athena/Items/Cosmetics/Backpacks/BID_506_DonutCup.BID_506_DonutCup" "(TagName=\"Cosmetics.Variant.Channel.Parts\")") "(TagName=\"Cosmetics.Variant.Property.Stage2\")")
        (return "(TagName=\"Cosmetics.Variant.Property.Mat2\")"))))
  (return (Class|FortCosmeticFunctionLibrary|GetActiveVariantForSoftItemAndChannel Loadout "/BRCosmetics/Athena/Items/Cosmetics/Backpacks/BID_506_DonutCup.BID_506_DonutCup" "(TagName=\"Cosmetics.Variant.Channel.Parts\")")))
